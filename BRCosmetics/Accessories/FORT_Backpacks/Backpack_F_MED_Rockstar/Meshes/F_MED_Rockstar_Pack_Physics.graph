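# Physics asset for the F_MED_Rockstar backpack cosmetic, setting up its star and tassel as a jointed, swinging chain.
# The pack body carries a large capsule for the bag's volume plus a tiny one, while every other bone has one tiny capsule.
# Constraints link pack, star and five tassel bones in series, each joint allowing narrow swing with twist locked.
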PhysicsAsset /BRCosmetics/Accessories/FORT_Backpacks/Backpack_F_MED_Rockstar/Meshes/F_MED_Rockstar_Pack_Physics
bodies {
 pack [
  {shapeType=Capsule, length=1, radius=1}
  {shapeType=Capsule, length=13.25, radius=16.75, center=<-0.07 19.3444 -27.1871>, rotation=<rot -89.1807 125.6771 -125.5825>}
 ]
 star [{shapeType=Capsule, length=1, radius=1, center=<0 -0.507 0>, rotation=<rot 90 0 0>}]
 tassle1 [{shapeType=Capsule, length=1, radius=1, rotation=<rot 90 0 0>}]
 tassle2 [{shapeType=Capsule, length=1, radius=1, rotation=<rot 90 0 0>}]
 tassle3 [{shapeType=Capsule, length=1, radius=1, rotation=<rot 90 0 0>}]
 tassle4 [{shapeType=Capsule, length=1, radius=1, rotation=<rot 90 0 0>}]
 tassle5 [{shapeType=Capsule, length=1, radius=1, rotation=<rot 90 0 0>}]
}
constraints [
 {bone1Name=star, bone2Name=pack, swing1LimitDegrees=15, swing2LimitDegrees=15, twistLimitDegrees=5, twistMotion=Locked}
 {bone1Name=tassle1, bone2Name=star, swing1LimitDegrees=10, swing2LimitDegrees=15, twistLimitDegrees=5, twistMotion=Locked}
 {bone1Name=tassle2, bone2Name=tassle1, swing1LimitDegrees=10, swing2LimitDegrees=15, twistLimitDegrees=5, twistMotion=Locked}
 {bone1Name=tassle3, bone2Name=tassle2, swing1LimitDegrees=10, swing2LimitDegrees=15, twistLimitDegrees=5, twistMotion=Locked}
 {bone1Name=tassle4, bone2Name=tassle3, swing1LimitDegrees=10, swing2LimitDegrees=15, twistLimitDegrees=5, twistMotion=Locked}
 {bone1Name=tassle5, bone2Name=tassle4, swing1LimitDegrees=10, swing2LimitDegrees=15, twistLimitDegrees=5, twistMotion=Locked}
]
