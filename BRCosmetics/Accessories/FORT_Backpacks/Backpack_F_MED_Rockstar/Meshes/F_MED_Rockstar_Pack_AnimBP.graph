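# Animation blueprint for the Rockstar backpack cosmetic that physically simulates the pack from its ref pose.
# It blends by crouch state between two RigidBody setups with different physics assets, then adds yaw offsets to star and tassel bones.
# Pelvis yaw relative to ref pose and skydive-floating state drive those offsets.
# The EventGraph overrides simulation gravity for swim sprinting and skydiving, using pelvis direction and Perlin-noise flutter in freefall.
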AnimBlueprint /BRCosmetics/Accessories/FORT_Backpacks/Backpack_F_MED_Rockstar/Meshes/F_MED_Rockstar_Pack_AnimBP
target_skeleton=&/F_MED_Rockstar_Pack_Skeleton
graphs [
 AnimGraph {nodes [
   n1 = RigidBody!(
    ComponentPose=Local To Component(LocalPose=Use cached pose 'Presim'())
    Alpha
    OverrideWorldGravity:Vector=Get Gravity Override B() -> (Gravity Override B:Vector)
    ExternalForce
   ) -> (Pose) {
    bOverrideWorldGravity=true
    overridePhysicsAsset=&/F_MED_Rockstar_Pack_Crouch_Physics
    alphaBoolBlend {blendOption=Linear}
    baseBoneRef {boneName=root}
    componentAppliedLinearAccClamp=<2000 2000 1200>
    componentLinearAccScale=<0.4 0.4 0.35>
    componentLinearVelScale=<0.4 0.4 0.35>
    overrideWorldGravity {z=-980}
   }
   n2 = "Transform (Modify) Bone - Bone: star"(ComponentPose=@n1, Alpha=1, Translation="0,0,0", Rotation="0,-5,0", Scale="1,1,1") {rotationMode=BMM_Additive, rotationSpace=BCS_BoneSpace, alphaBoolBlend {blendOption=Linear}, boneToModify {boneName=star}, rotation {yaw=-5}}
   n3 = RigidBody(ComponentPose=Local To Component(LocalPose=Use cached pose 'Presim'()), OverrideWorldGravity=Get Gravity Override B()) {
    bOverrideWorldGravity=true
    overridePhysicsAsset=&/F_MED_Rockstar_Pack_Physics
    alphaBoolBlend {blendOption=Linear}
    baseBoneRef {boneName=root}
    componentAppliedLinearAccClamp=<1200 1200 1200>
    componentLinearAccScale=<0.3 0.3 0.35>
    componentLinearVelScale=<0.3 0.3 0.35>
    overrideWorldGravity {z=-980}
   }
   n4 = Get spine01_alpha() -> (spine01_alpha:double)
   n5 = "Transform (Modify) Bone - Bone: tassle3"(ComponentPose=@n3, Alpha=@n4, Translation="0,0,0", Rotation="0.000,-15,0.000", Scale="1,1,1") {
    rotationMode=BMM_Additive
    rotationSpace=BCS_BoneSpace
    alphaBoolBlend {blendOption=Linear}
    boneToModify {boneName=tassle3}
    rotation {yaw=-15}
   }
   n6 = "Transform (Modify) Bone - Bone: tassle2"(ComponentPose=@n5, Alpha=@n4, Translation="0,0,0", Rotation="0.000,-15,0.000", Scale="1,1,1") {
    rotationMode=BMM_Additive
    rotationSpace=BCS_BoneSpace
    alphaBoolBlend {blendOption=Linear}
    boneToModify {boneName=tassle2}
    rotation {yaw=-15}
   }
   n7 = "Transform (Modify) Bone - Bone: tassle1"(ComponentPose=@n6, Alpha=@n4, Translation="0,0,0", Rotation="0.000,-15,0.000", Scale="1,1,1") {
    rotationMode=BMM_Additive
    rotationSpace=BCS_BoneSpace
    alphaBoolBlend {blendOption=Linear}
    boneToModify {boneName=tassle1}
    rotation {yaw=-15}
   }
   n8 = "Transform (Modify) Bone - Bone: star"!(ComponentPose=@n7, bAlphaBoolEnabled:bool=Get bIsSkydiveFloating(), Translation="0,0,0", Rotation="0,-15,0", Scale="1,1,1") -> (Pose) {
    alphaInputType=Bool
    rotationMode=BMM_Additive
    rotationSpace=BCS_BoneSpace
    alphaBoolBlend {blendOption=Linear}
    boneToModify {boneName=star}
    rotation {yaw=-15}
   }
   n9 = "Transform (Modify) Bone - Bone: tassle4"(
    ComponentPose=@n8
    Alpha=Get spine01_alpha_backwalk() -> (spine01_alpha_backwalk:double)
    Translation="0,0,0"
    Rotation="0,-10,0"
    Scale="1,1,1"
   ) {
    rotationMode=BMM_Additive
    rotationSpace=BCS_BoneSpace
    alphaBoolBlend {blendOption=Linear}
    boneToModify {boneName=tassle4}
    rotation {yaw=-10}
   }
   Output Pose(Result=Blend Poses by bool(BlendPose_0=Component To Local(ComponentPose=@n2), BlendPose_1=Component To Local(ComponentPose=@n9), bActiveValue=Get bIsCrouching()) {blendType=Linear}) {layerGroup=None}
   Save cached pose 'Presim'(Pose=Local Space Ref Pose()) {cacheName=Presim}
  ]}
 EventGraph {nodes [
   Event BlueprintInitializeAnimation()
   > Set Skel Mesh(Skel Mesh=GetPartSkeletalMeshComponent())
   > n1 = Get!(execute:Exec) -> (then:Exec, else:Exec, Skel Mesh:Skeletal Mesh Component Object Reference)
   Event BlueprintUpdateAnimation()
   > n2 = Sequence()
   > n3 = Branch(Condition=Get bIsCrouching())
   n4 = GetOwnerSkeletalMeshComponent()
   n5 = BreakRotator(InRot=BreakTransform(InTransform=GetDeltaTransformFromRefPose(self=@n4, BoneName=pelvis, BaseName=root)).Rotation)
   > Set spine01_alpha(spine01_alpha:double=MapRangeClamped(Value=@n5.Yaw, InRangeA=0, InRangeB=30)) -> (Output_Get:double)
   Set spine01_alpha(execute=@n3.else, spine01_alpha=MapRangeClamped(Value=@n5.Yaw, InRangeA=0, InRangeB=15))
   > Set spine01_alpha_backwalk(spine01_alpha_backwalk:double=MapRangeClamped(Value=@n5.Yaw, InRangeA=-3, InRangeB=-10)) -> (Output_Get:double)
   n6 = Branch(execute=@n2.then_1, Condition=Get bIsSwimSprinting() -> (bIsSwimSprinting:bool))
   > Set Gravity Override B(Gravity Override B:Vector="0, 0, 0") -> (Output_Get:Vector)
   n7 = Branch(execute=@n6.else, Condition=Get bIsSkyDiving())
   > n8 = Branch(Condition=Get bIsParachuteOpened())
   Set Gravity Override B(execute=[@n7.else, @n8.then], Gravity Override B="0, 0,-920")
   n9 = Branch(execute=@n8.else, Condition="float > float"(A=Get PawnVelocityZ(), B=50) {is=title K2Node_CallFunction})
   > Set Gravity Override B(Gravity Override B="0, 0,200")
   n10 = PerlinNoise1D!(Value:float=GetGameTimeInSeconds()) -> (ReturnValue:float)
   n11 = float * float(A="Sin (Degrees)"!(A:double=float * float(A=@n10, B=500)) -> (ReturnValue:double), B=-400)
   n12 = Get Skel Mesh()
   n13 = GetSocketRotation!(self:Scene Component Object Reference=@n12, InSocketName:Name=pelvis) -> (ReturnValue:Rotator)
   n14 = vector + vector(
    A:Vector=vector * float(A=GetForwardVector!(InRot:Rotator=@n13) -> (ReturnValue:Vector), B=-800)
    B:Vector=MakeVector(Y=0, Z=float + float(A=@n11, B=200) {is=title K2Node_CommutativeAssociativeBinaryOperator})
   ) -> (ReturnValue:Vector) {is=title K2Node_CommutativeAssociativeBinaryOperator}
   Set Gravity Override B(execute=@n9.else, Gravity Override B=@n14)
   TryGetPawnOwner()
   GetAnimInstance(self=@n1)
   n15 = GetSocketLocation!(self:Scene Component Object Reference=@n12, InSocketName:Name=pelvis) -> (ReturnValue:Vector)
   DrawDebugLine!(
    execute
    LineStart:Vector=@n15
    LineEnd:Vector=vector + vector(A=@n15, B=@n14)
    LineColor:Linear Color Structure=(R=1,G=0,B=0,A=1)
    Duration:float=0
    Thickness:float=3
    DepthPriority:EDrawDebugSceneDepthPriorityGroup Enum=World
   ) -> (then)
   Set Skel Mesh(Skel Mesh=@n12)
  ]}
]
variables [
 Gravity Override B {type=Vector, default=<0 0 -920>}
 Skel Mesh {type=SkeletalMeshComponent}
 spine01_alpha {type=number}
 spine01_alpha_backwalk {type=number}
]
